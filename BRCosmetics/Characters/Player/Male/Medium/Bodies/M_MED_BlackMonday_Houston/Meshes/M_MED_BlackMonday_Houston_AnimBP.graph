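# Animation Blueprint for the Black Monday Houston medium male body that copies its pose from the attached parent mesh and adjusts the body for a matching cape backpack.
# On initialize it caches the owning Athena player pawn, then a Do Once during update caches that pawn's backpack skeletal mesh.
# Each update, if that mesh has a black_monday_cape socket, it sets the Body_back_morph morph target to full, with no reset path.
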AnimBlueprint /BRCosmetics/Characters/Player/Male/Medium/Bodies/M_MED_BlackMonday_Houston/Meshes/M_MED_BlackMonday_Houston_AnimBP
target_skeleton=&/M_MED_BlackMonday_Houston_Skeleton
graphs [
 AnimGraph {nodes [Output Pose(Result=Copy Pose From Mesh() {bUseAttachedParent=true}) {layerGroup=None}]}
 EventGraph {nodes [
   Event BlueprintUpdateAnimation()
   > n1 = Sequence()
   > Do Once()
   > n2 = Get!(execute:Exec) -> (then:Exec, else:Exec, player_pawn_athena:Fort Player Pawn Athena Object Reference)
   n3 = GetSkeletalMeshForPartType(self=@n2, PartType=Backpack)
   > Is Valid(InputObject=@n3)
   > Set backpack_skelmesh(backpack_skelmesh:Skeletal Mesh Component Object Reference=@n3) -> (Output_Get:Skeletal Mesh Component Object Reference)
   n4 = Get backpack_skelmesh() -> (backpack_skelmesh:Skeletal Mesh Component Object Reference)
   Is Valid(exec=@n1.then_1, InputObject=@n4)
   > Branch(Condition=DoesSocketExist!(self:Scene Component Object Reference=@n4, InSocketName:Name=black_monday_cape) -> (ReturnValue:bool))
   > SetMorphTarget!(execute, self:Anim Instance Object Reference, MorphTargetName:Name=Body_back_morph, Value:float=1) -> (then)
   Event BlueprintInitializeAnimation()
   n5 = TryGetPawnOwner()
   > Is Valid(InputObject=@n5)
   > n6 = Cast To FortPlayerPawnAthena(Object=@n5)
   > Set player_pawn_athena(player_pawn_athena:Fort Player Pawn Athena Object Reference=@n6) -> (Output_Get:Fort Player Pawn Athena Object Reference)
  ]}
]
variables [
 backpack_skelmesh {type=SkeletalMeshComponent}
 player_pawn_athena {type=FortPlayerPawnAthena}
]
