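# Physics asset of cape collision shapes for the M_MED_BlackMonday_Houston male medium player body, by its name for the cape to collide against.
# Capsules cover pelvis, spine, clavicles and neck, with an oversized one offset along neck_02 and three overlapping on spine_02.
# Arm, leg and foot bodies are present but hold no shapes.
# It defines no constraints, so it sets up no jointed ragdoll.
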PhysicsAsset /BRCosmetics/Characters/Player/Male/Medium/Bodies/M_MED_BlackMonday_Houston/Meshes/M_MED_BlackMonday_Houston_CapeColliders
bodies {
 calf_l []
 calf_r []
 clavicle_l [{shapeType=Capsule, length=7, radius=10.5, center=<9.9177 1.387 -2.9891>, rotation=<rot 75.9352 -0.0001 175.5622>}]
 clavicle_r [{shapeType=Capsule, length=7, radius=10.5, center=<-10.1321 1.4615 3.0429>, rotation=<rot 75.9351 -0.0001 -4.4378>}]
 foot_l []
 foot_r []
 lowerarm_l []
 lowerarm_r []
 neck_01 [{shapeType=Capsule, length=6.005, radius=10.255, center=<0 2.6796 0.0003>, rotation=<rot 90 0 0>}]
 neck_02 [{shapeType=Capsule, length=29, radius=23.005, shapeName=neck_02_capsule, center=<25.8995 -2.0334 0>, rotation=<rot 0 -90 0>}]
 pelvis [{shapeType=Capsule, length=8.6562, radius=13.5388, center=<-2.6215 -1.3695 0>}]
 spine_01 [{shapeType=Capsule, length=15.6986, radius=13.6565, center=<-2.5767 -2.1041 0>}]
 spine_02 [
  {shapeType=Capsule, length=15.5736, radius=12.1877, center=<-2.6215 -0.2296 0.0001>}
  {shapeType=Capsule, length=29.875, radius=11.1013, center=<2.5024 0.4992 3.9825>, rotation=<rot 90 0.0049 0.0049>}
  {shapeType=Capsule, length=29.875, radius=11.1013, center=<2.4704 0.4991 -3.9825>, rotation=<rot 90 -0.0396 -0.0396>}
 ]
 spine_03 [{shapeType=Capsule, length=14.1361, radius=12.0002, center=<-1.953 -0.4085 0.0001>}]
 spine_04 [{shapeType=Capsule, length=11.6361, radius=13.6252, center=<2.1467 1.2063 0.0002>}]
 spine_05 [{shapeType=Capsule, length=22.2812, radius=13.0625, center=<-3.1588 -0.9828 0>}]
 thigh_l []
 thigh_r []
 upperarm_l []
 upperarm_r []
}
constraints []
